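# Lit, Nanite-enabled material for TV wall screen props, showing one of four animated static patterns as tinted emissive over black.
# Screens offer flipbook noise, noise with moving dark lines, scrolling static with white lines, and horizontally glitched noise, with vignette masks.
# A static switch picks the pattern by parameter or by object position plus vertex red, resolved through chained Sign-driven lerps.
# Object-position sine hashes vary line and glitch timing.
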Material /Game/Environments/Artemis/Props/RealityPG/RealityPG_TvWall_A/Materials/M_RealityPG_TvWallScreens_A
graph {
 properties {bUseMaterialAttributes=true, bUsedWithNanite=true, pixelDepthOffsetMode=PDOM_Legacy, usedShadingModels=Default Lit}
 nodes [
  n1 = ObjectPositionWS()
  n2 = DotProduct(A=@n1, B=Constant3Vector() {constant=<rgba 158.1781 280.8795 614.0283 0>})
  Frac(None=Multiply(A=Sine(None=@n2)) {constB=10000})
  n3 = Step(Y=MaterialFunctionCall(Float2=TextureCoordinate()) {materialFunction=&BreakOutFloat2Components}.G) {constX=0.5}
  n4 = Time()
  n5 = Sine(None=Frac(None=Add(A=@n4)))
  LinearInterpolate(A=@n3, B=OneMinus(None=@n3), Alpha=Saturate(None=Sign(None=@n5))) {constB=0}
  n6 = TextureSample(UVs=TextureCoordinate()) {
   samplerSource=SSM_Wrap_WorldGroupSettings
   samplerType=SAMPLERTYPE_Masks
   texture=/Game/Missions/Primary/RepairTheShelter/Textures/Cinematic_Screen_Static
  }
  n7 = OneMinus(None=ScalarParameter() {defaultValue=1, group=Screen2, parameterName="VignetOpacity(Screen2)"})
  NamedRerouteDeclaration(None=Saturate(None=Add(A=@n6.A, B=@n7))) {name="vignet(Screen2)", nodeColor=<rgba 0.4297 1 0 1>}
  NamedRerouteDeclaration(None=Saturate(None=@n6.A)) {name="vignet(Screen3)", nodeColor=<rgba 0.4297 1 0 1>}
  n8 = OneMinus(None=ScalarParameter() {defaultValue=1, group=Screen4, parameterName="VignetOpacity(Screen4)"})
  NamedRerouteDeclaration(None=Saturate(None=Add(A=@n6.A, B=@n8))) {name="vignet(Screen4)", nodeColor=<rgba 0.4297 1 0 1>}
  n9 = OneMinus(None=ScalarParameter() {defaultValue=1, group=Screen2, parameterName="VignetOpacity(Screen1)"})
  NamedRerouteDeclaration(None=Saturate(None=Add(A=@n6.A, B=@n9))) {name="vignet(Screen1)", nodeColor=<rgba 0.4297 1 0 1>}
  n10 = Time()
  n11 = Add(A=Frac(None=Multiply(A=@n10) {constB=3}), B=WorldPosition() {worldPositionShaderOffset=WPT_ExcludeAllShaderOffsets}.XYZ)
  Noise!(World Position=@n11, FilterWidth)
  n12 = MaterialFunctionCall(Float3=@n1) {materialFunction=&BreakOutFloat3Components}
  n13 = Abs(None=@n12.R)
  n14 = Abs(None=@n12.G)
  n15 = Divide(A=Add(A=Add(A=@n13, B=@n14), B=Abs(None=@n12.B))) {constB=3}
  n16 = ScalarParameter() {defaultValue=50, parameterName=ScreenTransitionDistance}
  n17 = Fmod(A=Round(None=Divide(A=@n15, B=@n16) {constB=3}), B=Constant() {r=4})
  n18 = VertexColor()
  n19 = Round(None=Multiply(A=Subtract(A=@n18.R) {constB=0.25}) {constB=4})
  n20 = ScalarParameter() {defaultValue=1, parameterName="ScreenPatternNumber(1,2,3,4)", sliderMax=4, sliderMin=1}
  n21 = StaticSwitchParameter(True=Fmod(A=Add(A=@n17, B=@n19), B=Constant() {r=4}), False=Subtract(A=Round(None=@n20))) {parameterName=UseVertexColorforPattern}
  n22 = Subtract(A=@n21) {constB=0}
  n23 = Time()
  n24 = ScalarParameter() {defaultValue=1, group=Screen1, parameterName=NoiseFlickeringFrequency}
  n25 = Constant() {r=5}
  n26 = TextureCoordinate()
  n27 = ScalarParameter() {defaultValue=5, group=Screen1, parameterName=NoiseTiling}
  n28 = Multiply(
   A=MaterialFunctionCall(
    "Animation  Phase (0-1)"=Multiply(A=@n23, B=@n24)
    Number of Rows=@n25
    Number of Columns=@n25
    Texture=TextureObject() {texture=/Game/Global/Textures/Default/Noise/Noise}
    UVs=Multiply(A=@n26, B=@n27)
   ) {materialFunction=&FlipBook}.Result
   B=NamedRerouteUsage()
  )
  n29 = ScalarParameter() {defaultValue=1, group=Screen1, parameterName=Screen1FinalEmsive}
  n30 = Constant() {r=5}
  n31 = MaterialFunctionCall(
   "Animation  Phase (0-1)"=Multiply(A=Time(), B=ScalarParameter() {defaultValue=1, group=Screen2, parameterName="NoiseFlickeringFrequency(screen2)"})
   Number of Rows=@n30
   Number of Columns=@n30
   Texture=TextureObject() {texture=&Noise}
   UVs=Multiply(A=TextureCoordinate(), B=ScalarParameter() {defaultValue=5, group=Screen2, parameterName="NoiseTiling(screen2)"})
  ) {materialFunction=&FlipBook}
  n32 = PreSkinnedPosition()
  n33 = PreSkinnedLocalBounds()
  n34 = MaterialFunctionCall(Float3=Divide(A=VertexInterpolator(VS=@n32).PS, B=Subtract(A=@n33.Max, B=@n33.Max))) {materialFunction=&BreakOutFloat3Components}
  n35 = VertexColor()
  n36 = Multiply(A=LinearInterpolate(A=@n34.B, B=Multiply(A=@n34.G) {constB=5}, Alpha=Sign(None=@n35.G)), B=Constant() {r=-3}) {constB=80}
  n37 = Sine(None=DotProduct(A=ObjectPositionWS(), B=Constant3Vector() {constant=<rgba 158.1781 34 614.0283 0>}))
  n38 = Add(A=@n36, B=Frac(None=Multiply(A=@n37) {constB=10000}))
  n39 = Time()
  n40 = ScalarParameter() {defaultValue=0.5, group=Screen2, parameterName=Line1Speed}
  n41 = Constant()
  n42 = SphereMask(
   A=Frac(None=@n38)
   B=Frac(None=Multiply(A=@n39, B=@n40))
   Radius=ScalarParameter() {defaultValue=0.07, group=Screen2, parameterName=Line1Thickness}
   Hardness=@n41
  )
  n43 = ScalarParameter() {defaultValue=1, group=Screen2, parameterName=Line1Opacity}
  n44 = ScalarParameter() {defaultValue=2, group=Screen2, parameterName=Line2Numer}
  n45 = Time()
  n46 = ScalarParameter() {defaultValue=0.5, group=Screen2, parameterName=Line2Speed}
  n47 = SphereMask(
   A=Frac(None=Multiply(A=@n38, B=@n44) {constB=2})
   B=Frac(None=Multiply(A=@n45, B=@n46) {constB=0.5})
   Radius=Multiply(
    A=ScalarParameter() {defaultValue=0.035, group=Screen2, parameterName=Line2Thickness}
    B=ScalarParameter() {defaultValue=2, group=Screen2, parameterName=Line2Numer}
   ) {constB=2}
   Hardness=@n41
  )
  n48 = ScalarParameter() {defaultValue=0.85, group=Screen2, parameterName=Line2Opacity}
  n49 = Clamp(None=Add(A=Multiply(A=@n42, B=@n43), B=Multiply(A=@n47, B=@n48) {constB=0.85})) {maxDefault=0.85}
  n50 = Multiply(A=Multiply(A=Multiply(A=@n31.Result) {constB=0.25}, B=OneMinus(None=@n49)), B=NamedRerouteUsage())
  n51 = ScalarParameter() {defaultValue=1, group=Screen2, parameterName=Screen2FinalEmsive}
  n52 = LinearInterpolate(
   A=Multiply(A=Multiply(A=@n28, B=@n29), B=VectorParameter() {group=Screen1, parameterName=Screen1FinalTint, defaultValue=<rgba 1 1 1 1>}.RGB)
   B=Multiply(A=Multiply(A=@n50, B=@n51), B=VectorParameter() {group=Screen2, parameterName=Screen2FinalTint, defaultValue=<rgba 1 1 1 1>}.RGB)
   Alpha=Saturate(None=Sign(None=@n22))
  )
  n53 = Saturate(None=Sign(None=Subtract(A=@n21)))
  n54 = Constant() {r=3}
  n55 = ScalarParameter() {defaultValue=0.1, group=Screen3, parameterName=TexturedLineSpeed}
  n56 = TextureSample(UVs=Panner(Coordinate=TextureCoordinate(), Speed=Multiply(A=@n55) {constB=-1}) {speedY=-0.1}) {
   samplerSource=SSM_Wrap_WorldGroupSettings
   samplerType=SAMPLERTYPE_Masks
   texture=&Cinematic_Screen_Static
  }
  n57 = NamedRerouteUsage()
  n58 = LinearInterpolate(
   B=MaterialFunctionCall(
    Number of Rows=@n54
    Number of Columns=@n54
    Texture=TextureObject() {samplerType=SAMPLERTYPE_LinearColor, texture=/Game/Global/Textures/Default/Noise/T_Apollo_TCNoise}
    UVs=TextureCoordinate() {uTiling=3, vTiling=3}
   ) {materialFunction=&FlipBook}.Result
   Alpha=Add(A=@n56.R, B=Multiply(A=@n57) {constB=0.05})
  ) {constA=0.001}
  n59 = Frac(None=Multiply(A=@n38, B=ScalarParameter() {defaultValue=2, group=Screen3, parameterName=WhiteLineNumer}) {constB=2})
  n60 = Frac(None=Multiply(A=Time(), B=ScalarParameter() {defaultValue=0.5, group=Screen3, parameterName=WhiteLineSpeed}) {constB=0.5})
  n61 = Multiply(
   A=ScalarParameter() {defaultValue=0.035, group=Screen3, parameterName=WhiteLineThickness}
   B=ScalarParameter() {defaultValue=2, group=Screen3, parameterName=WhiteLineNumer}
  ) {constB=2}
  n62 = Multiply(A=Multiply(A=SphereMask(A=@n59, B=@n60, Radius=@n61)) {constB=0.85}) {constB=0.05}
  n63 = Multiply(
   A=Multiply(A=Add(A=@n58, B=@n62), B=ScalarParameter() {defaultValue=1, group=Screen3, parameterName=Screen3FinalEmsive})
   B=VectorParameter() {group=Screen3, parameterName=Screen3FinalTint, defaultValue=<rgba 1 1 1 1>}.RGB
  )
  n64 = Sign(None=Subtract(A=@n21) {constB=2})
  n65 = MaterialFunctionCall(Float2=Panner(Coordinate=TextureCoordinate() {uTiling=0.5, vTiling=0.5}) {speedY=-0.05}) {materialFunction=&BreakOutFloat2Components}
  n66 = Add(A=@n65.R, B=ScalarParameter() {defaultValue=0.1, group=Screen4, parameterName=GlitchStrength}) {constB=0.1}
  n67 = Multiply(A=TextureCoordinate(), B=ScalarParameter() {defaultValue=10, group=Screen4, parameterName=GlitchTiling})
  n68 = Frac(None=Time())
  n69 = Sine(None=Multiply(A=Ceil(None=@n36), B=Constant() {r=11.5142}))
  n70 = Sine(None=DotProduct(A=ObjectPositionWS(), B=Constant3Vector() {constant=<rgba 158.1781 34 614.0283 0>}))
  n71 = Frac(None=Add(A=Multiply(A=@n69) {constB=10000}, B=Multiply(A=@n70) {constB=10000}))
  n72 = ScalarParameter() {defaultValue=10, group=Screen4, parameterName=GlitchFrequency}
  n73 = Ceil(None=Multiply(A=Add(A=@n68, B=@n71), B=@n72))
  n74 = ConstantBiasScale(None=Sine(None=Divide(A=@n73, B=@n72)))
  n75 = LinearInterpolate(
   B=Sine(None=MaterialFunctionCall(Float2=@n67) {materialFunction=&BreakOutFloat2Components}.G)
   Alpha=Frac(None=Multiply(A=@n74) {constB=10000})
  )
  n76 = TextureSample(UVs=AppendVector(A=LinearInterpolate(A=@n66, B=@n65.R, Alpha=@n75), B=@n65.G)) {samplerSource=SSM_Wrap_WorldGroupSettings, samplerType=SAMPLERTYPE_LinearColor, texture=&T_Apollo_TCNoise}
  n77 = Multiply(A=Multiply(A=@n76.RGB, B=NamedRerouteUsage()), B=ScalarParameter() {defaultValue=1, group=Screen4, parameterName=Screen4FinalEmsive})
  n78 = VectorParameter() {group=Screen4, parameterName=Screen4FinalTint, defaultValue=<rgba 1 1 1 1>}
  SetMaterialAttributes!(
   MaterialAttributes
   Emissive Color=LinearInterpolate(A=LinearInterpolate(A=@n52, B=@n63, Alpha=@n53), B=Multiply(A=@n77, B=@n78.RGB), Alpha=Saturate(None=@n64))
   Base Color=Constant()
   Metallic=ScalarParameter() {parameterName=Metallic}
   Specular=ScalarParameter() {parameterName=Specular}
   Roughness=ScalarParameter() {defaultValue=0.5, parameterName=Roughness}
   Normal=TextureSampleParameter2D() {parameterName=Normals, samplerType=SAMPLERTYPE_Normal, texture=&FlatNormalBlue}.RGB
  ) {attributeSetTypes [
    #1
    #2
    #3
    #4
    #5
    #6
   ]}
  VectorNoise(World Position=@n11) {noiseFunction=VNF_VectorALU}
 ]
}
